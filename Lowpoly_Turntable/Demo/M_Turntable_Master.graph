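# M_Turntable_Master is a master material for the Lowpoly Turntable demo, using the Default Lit shading model to present a model's wireframe, ambient occlusion and normals.
# It multiplies the RGB of the Wireframe and AO texture parameters, overlaying edge lines on baked occlusion.
# A separate Normals parameter samples a normal map.
# All three parameters default to Mannequin body textures and can be swapped in material instances.
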Material /Lowpoly_Turntable/Demo/M_Turntable_Master
graph {
 properties {pixelDepthOffsetMode=PDOM_Legacy, usedShadingModels=Default Lit}
 nodes [
  Multiply(
   A=TextureSampleParameter2D() {parameterName=Wireframe, texture=&/Mannequin_Body_wireframe}.RGB
   B=TextureSampleParameter2D() {parameterName=AO, texture=&/Mannequin_Body_ao}.RGB
  )
  TextureSampleParameter2D() {parameterName=Normals, samplerType=SAMPLERTYPE_Normal, texture=&/Mannequin_Body_normal}
 ]
}
